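; Parent Blueprint for the SilverPopcorn large camera crane prop, assembled and posed in the construction script; its event graph is empty.
; It checks for static mesh components with no mesh assigned, printing a warning, then snaps base bracket, arm, camera bracket and camera to mesh sockets in chain order.
; Variables drive body rotation, clamped arm angle and bracket rotation, while the camera gets a fixed yaw offset and uniform scale.
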(event EventBeginPlay)

(event Collision|EventActorBeginOverlap
  (params (OtherActor :type Actor%20Object%20Reference)))

(event EventTick
  (params (DeltaSeconds :type Float%20%28single-precision%29)))

(fn SetCameraBracketTransforms
  (bind _cranearm (Variables|Default|GetCraneArm))
  (bind _returnvalue (Transformation|GetAllSocketNames _cranearm))
  (bind _sm_silverpopcorn_largecranecamerabracket_a (Variables|Default|GetSM_SilverPopcorn_LargeCraneCameraBracket_A))
  (Utilities|IsValid _sm_silverpopcorn_largecranecamerabracket_a
    (:Is%20Valid
      (if (> (Utilities|Array|Length _returnvalue) 0)
        (Transformation|SetRelativeTransform _sm_silverpopcorn_largecranecamerabracket_a (Math|Transform|MakeTransform (.location (Transformation|GetSocketTransform (Variables|Default|GetCraneArm) (Utilities|Array|Get(acopy) _returnvalue) "RTS_Actor")) "0, 0, 0" (.scale (Transformation|GetSocketTransform (Variables|Default|GetCraneArm) (Utilities|Array|Get(acopy) _returnvalue) "RTS_Actor"))) false false)
        (Transformation|AddLocalRotation (Variables|Default|GetSM_SilverPopcorn_LargeCraneCameraBracket_A) (Math|Rotator|MakeRotator 0.0 0.0 (* (Variables|Default|GetCraneCameraRotation) -1.0)) false false)))))

(fn SetCameraTransforms
  (bind _sm_silverpopcorn_largecranecamerabracket_a (Variables|Default|GetSM_SilverPopcorn_LargeCraneCameraBracket_A))
  (bind _returnvalue (Transformation|GetAllSocketNames _sm_silverpopcorn_largecranecamerabracket_a))
  (bind _sm_silverpopcorn_dollycamera_b (Variables|Default|GetSM_SilverPopcorn_DollyCamera_B))
  (Utilities|IsValid _sm_silverpopcorn_dollycamera_b
    (:Is%20Valid
      (if (> (Utilities|Array|Length _returnvalue) 0)
        (Transformation|SetRelativeTransform _sm_silverpopcorn_dollycamera_b (Transformation|GetSocketTransform _sm_silverpopcorn_largecranecamerabracket_a (Utilities|Array|Get(acopy) _returnvalue) "RTS_Actor") false false)
        (Transformation|AddLocalRotation (Variables|Default|GetSM_SilverPopcorn_DollyCamera_B) "0.000000,90.000000,0.000000" false false)
        (Transformation|SetRelativeScale3D (Variables|Default|GetSM_SilverPopcorn_DollyCamera_B) (Math|Vector|MakeVector (LocalVariables|GetL_DollyCamera_Scale) (LocalVariables|GetL_DollyCamera_Scale) (LocalVariables|GetL_DollyCamera_Scale)))))))

(fn SetCraneArmTransforms
  (bind _sm_silverpopcorn_largecranebasebracket_a (Variables|Default|GetSM_SilverPopcorn_LargeCraneBaseBracket_A))
  (bind _returnvalue (Transformation|GetAllSocketNames _sm_silverpopcorn_largecranebasebracket_a))
  (bind _cranearm (Variables|Default|GetCraneArm))
  (bind _cranearmangle (Variables|Default|GetCraneArmAngle))
  (Utilities|IsValid _cranearm
    (:Is%20Valid
      (if (> (Utilities|Array|Length _returnvalue) 0)
        (Transformation|SetRelativeTransform _cranearm (Transformation|GetSocketTransform _sm_silverpopcorn_largecranebasebracket_a (Utilities|Array|Get(acopy) _returnvalue) "RTS_Actor") false false)
        (if (<= _cranearmangle 8e+01)
          (if (not (>= _cranearmangle -2e+01))
            (Variables|Default|SetCraneArmAngle -2e+01))
          (else
            (Variables|Default|SetCraneArmAngle 8e+01)))
        (Transformation|SetRelativeRotation (Variables|Default|GetCraneArm) (Math|Rotator|MakeRotator (Variables|Default|GetCraneArmAngle) (.pitch (Transformation|GetWorldRotation (Variables|Default|GetCraneArm))) (.yaw (Transformation|GetWorldRotation (Variables|Default|GetCraneArm)))) false false)))))

(fn SetCraneAssemblyRotation
  (bind _staticmeshcomponent (Variables|Debug|GetStaticMeshComponent))
  (bind _returnvalue (Transformation|GetAllSocketNames _staticmeshcomponent))
  (bind _sm_silverpopcorn_largecranebasebracket_a (Variables|Default|GetSM_SilverPopcorn_LargeCraneBaseBracket_A))
  (if (Utilities|Array|IsNotEmpty _returnvalue)
    (Transformation|SetRelativeLocation _sm_silverpopcorn_largecranebasebracket_a (.location (Transformation|GetSocketTransform _staticmeshcomponent (Utilities|Array|Get(acopy) _returnvalue) "RTS_Component")) false false)
    (Transformation|AddRelativeRotation _sm_silverpopcorn_largecranebasebracket_a (Math|Rotator|MakeRotator (.roll (Variables|/Script/Engine.SceneComponent|GetRelativeRotation _sm_silverpopcorn_largecranebasebracket_a)) (.pitch (Variables|/Script/Engine.SceneComponent|GetRelativeRotation _sm_silverpopcorn_largecranebasebracket_a)) (Variables|Default|GetCraneBodyRotation)) false false)))

(fn ConstructionScript
  (|CallParentFunction)
  (CallFunction|ValidateMeshComponents)
  (CallFunction|SetCraneAssemblyRotation)
  (CallFunction|SetCraneArmTransforms)
  (CallFunction|SetCameraBracketTransforms)
  (CallFunction|SetCameraTransforms))

(fn ValidateMeshComponents
  (for _array_element (Actor|GetComponentsByClass "/Script/Engine.StaticMeshComponent")
    (Utilities|IsValid (Variables|/Script/Engine.StaticMeshComponent|GetStaticMesh _array_element)
      (:Is%20Not%20Valid
        (Development|PrintString (Utilities|String|Append "Assign a mesh to the" (.RightS (Utilities|String|Split (Utilities|GetDisplayName _array_element) "." "IgnoreCase" "FromEnd")) " component"))))))
